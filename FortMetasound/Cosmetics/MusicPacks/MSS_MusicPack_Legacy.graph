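# Stereo MetaSound source for legacy music packs, playing an optional intro, then a loop wave after IntroDelay.
# Start time either seeks into the intro and shortens the delay, or skips to an offset in the loop.
# An ADSR envelope and lobby music multiplier drive mixer gain; non-looping playback fades out before the loop wave ends, firing OnFinished.
# Outputs report total duration and playback time across both segments.
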MetaSoundSource /FortMetasound/Cosmetics/MusicPacks/MSS_MusicPack_Legacy
inputs [
 bLooping {type=Bool}
 Fade Out Time {type=Time}
 Input {type=Float}
 Intro {type=WaveAsset, access=Constructor}
 IntroDelay {type=Time, access=Constructor}
 Loop {type=WaveAsset, access=Constructor}
 UE.Source.OnPlay {type=Trigger}
 UE.Source.StartTime.StartTime {type=Time}
]
interfaces [UE.OutputFormat.Stereo:1.0, UE.Source.OneShot:1.0, UE.Source.StartTime:1.0, UE.Source:1.1]
outputs [
 Duration {type=Time}
 OnLoopPoint {type=Trigger}
 Playback Time {type=Time}
 "UE.OutputFormat.Stereo.Audio:0" {type=Audio}
 "UE.OutputFormat.Stereo.Audio:1" {type=Audio}
 UE.Source.OneShot.OnFinished {type=Trigger}
]
pages [Default {
  input_defaults {
   Fade Out Time=0.5
   Input=0
   Intro=null
   IntroDelay=4.114
   Loop=null
   UE.Source.OnPlay=null
   UE.Source.StartTime.StartTime=0
   bLooping="<MetaSoundBuilderResult.SUCCEEDED: 0>"
  }
  nodes [
   GetWaveDuration_1 = "UE.Get Wave Duration: "(Wave=^Loop)
   Add_1 = UE.Add:Time(PrimaryOperand=^IntroDelay, AdditionalOperands=@GetWaveDuration_1.Duration)
   n1 = VariableAccessor.Float(Variable=InitVariable.Float().Variable)
   n2 = InitVariable.Float()
   n3 = VariableAccessor.Time(Variable=VariableMutator.Time(Value=^UE.Source.StartTime.StartTime, Variable=InitVariable.Time().Variable).Variable)
   n4 = VariableAccessor.Time(Variable=@n3.Variable)
   n5 = /FortMetasound/Patches/IsWaveAssetValid(WaveAsset=^Intro, Trigger In=^UE.Source.OnPlay)
   n6 = TriggerCompare.Clamp:Float(
    B=UE.ConversionTimeToFloat(In=^IntroDelay).Out
    A=UE.ConversionTimeToFloat(In=VariableAccessor.Time(Variable=@n4.Variable).Value).Out
    Compare=@n5.Is Valid
   ) {also {defaults {Type=4}}}
   n7 = UE.Reroute(Value=@n6.True)
   TriggerDelay_0 = UE.Trigger Delay(In=@n7.Value, Delay Time=UE.Subtract:Time(PrimaryOperand=^IntroDelay, AdditionalOperands=@n4.Value) {also {version [1,1]}}.Out)
   TriggerAnyTriggerAny3_0 = "TriggerAny.Trigger Any (3)"(In 2=UE.Reroute(Value=@n6.False).Value, In 1=UE.Reroute(Value=@n5.Is Not Valid).Value, In 0=@TriggerDelay_0.Out)
   n8 = UE.Subtract:Time(AdditionalOperands=^IntroDelay, PrimaryOperand=^UE.Source.StartTime.StartTime) {also {version [1,1]}}
   n9 = UE.ConversionFloatToTime(In=Max.Max:Float(A=UE.ConversionTimeToFloat(In=@n8.Out).Out).Value)
   n10 = InitVariable.Time()
   WavePlayer_1 = UE.Wave Player:Stereo(
    Wave Asset=^Loop
    Play=@TriggerAnyTriggerAny3_0.Out
    Start Time=VariableAccessor.Time(Variable=VariableMutator.Time(Value=@n9.Out, Variable=@n10.Variable).Variable).Value
   ) {also {defaults {Loop="<MetaSoundBuilderResult.SUCCEEDED: 0>"}, version [1,1]}}
   n11 = UE.Reroute(Value=@WavePlayer_1.On Play)
   n12 = UE.Reroute(Value=@n11.Value)
   GetWaveDuration_0 = "UE.Get Wave Duration: "(Wave=^Loop)
   TriggerDelay_1 = UE.Trigger Delay(
    In=&MSP_Branch(Value=^bLooping, In=@n12.Value).False
    Delay Time=UE.Subtract:Time(AdditionalOperands=^Fade Out Time, PrimaryOperand=@GetWaveDuration_0.Duration) {also {version [1,1]}}.Out
   )
   ADSREnvelopeADSREnvelope_0 = ADSR Envelope.ADSR Envelope:Float(Trigger Attack=UE.Reroute(Value=@n12.Value).Value, Release Time=UE.Reroute(Value=^Fade Out Time).Value, Trigger Release=@TriggerDelay_1.Out) {also {defaults {Attack Time=0, Decay Time=0, Sustain Level=1}}}
   n13 = InitVariable.Float()
   n14 = UE.Multiply:Float(
    PrimaryOperand=UE.Multiply:Float(
     AdditionalOperands=/Game/Sounds/Modulation/Buses/Sliders/MSP_LobbyMusicMultiplier(FrontEndAttenuationDecibels=@n1.Value).NormalizedLobbyMusicSlider
     PrimaryOperand=VariableAccessor.Float(Variable=@n2.Variable).Value
    ).Out
    AdditionalOperands=VariableAccessor.Float(Variable=VariableMutator.Float(Value=@ADSREnvelopeADSREnvelope_0.Out Envelope, Variable=@n13.Variable).Variable).Value
   )
   WavePlayer_0 = UE.Wave Player:Stereo(Wave Asset=^Intro, Play=@n7.Value, Start Time=@n3.Value) {also {version [1,1]}}
   AudioMixerAudioMixerStereo2_0 = "AudioMixer.Audio Mixer (Stereo, 2)"(
    Gain 0=@n14.Out
    Gain 1=@n14.Out
    In 0 L=@WavePlayer_0.Out Left
    In 0 R=@WavePlayer_0.Out Right
    In 1 L=@WavePlayer_1.Out Left
    In 1 R=@WavePlayer_1.Out Right
   )
   n15 = UE.Reroute(Value=@WavePlayer_1.Playback Time)
   TriggerRouteTriggerRouteTime2_0 = "TriggerRoute.Trigger Route (Time, 2):Time"(
    Value 1=UE.Add:Time(AdditionalOperands=^IntroDelay, PrimaryOperand=UE.Reroute(Value=@n15.Value).Value).Out
    Set 0=UE.Reroute(Value=UE.Reroute(Value=@WavePlayer_0.On Play).Value).Value
    Set 1=UE.Reroute(Value=UE.Reroute(Value=@n11.Value).Value).Value
    Value 0=UE.Reroute(Value=UE.Reroute(Value=@WavePlayer_0.Playback Time).Value).Value
   )
  ]
  links [
   {dst="@UE.Source.OneShot.OnFinished", src=ADSREnvelopeADSREnvelope_0.On Done}
   {dst="@Duration", src=Add_1.Out}
   {dst="@UE.OutputFormat.Stereo.Audio:0", src=AudioMixerAudioMixerStereo2_0.Out L}
   {dst="@UE.OutputFormat.Stereo.Audio:1", src=AudioMixerAudioMixerStereo2_0.Out R}
   {dst="@Playback Time", src=TriggerRouteTriggerRouteTime2_0.Value}
  ]
 }]
properties {
 assetImportData=&self:AssetImportData
 audioPropertiesSheet=&/MSS_MusicPack_Legacy_GeneratedPropertySheet
 numChannels=2
 outputFormat=Stereo
 sampleRate=48000
 soundClassObject=/Game/Sounds/FortSoundClassesAndModes/Fort_Music_MusicPacks
 virtualizationMode=PlayWhenSilent
 soundSubmixSends [{}]
}
